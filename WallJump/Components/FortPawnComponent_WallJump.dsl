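; Wall-jump pawn component that overrides when targeting is allowed and which direction a player jumps.
; Targeting requires all must-have gameplay tags and none of the preventing tags on the player pawn.
; Jump direction uses normalized movement acceleration when within a maximum angle of the incoming direction.
; Otherwise it uses the parent's result, mirrored on the pawn's local X axis when acceleration points backward; zero acceleration keeps the parent's result.
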(event EventBeginPlay)

(event EventTick
  (params (DeltaSeconds :type Float%20%28single-precision%29)))

(fn BP_IsTargetingAllowed
  (bind _asfort_player_pawn (Utilities|Casting|CastToFortPlayerPawn (CallFunction|GetOwnerPawn)))
  (return (and (not (GameplayTags|HasAnyMatchingGameplayTags _asfort_player_pawn (Variables|Default|GetPreventTargetingTags))) (GameplayTags|HasAllMatchingGameplayTags _asfort_player_pawn (Variables|Default|GetMustHaveTargetingTags)))))

(fn CalcJumpDirection
  (params (InDirection :type Vector%20%28by%20ref%29 :default "0, 0, 0"))
  (bind _asfort_player_pawn (Utilities|Casting|CastToFortPlayerPawn (CallFunction|GetOwnerPawn)))
  (if (Math|Vector|VectorIsNearlyZero (Math|Vector|Normalize (Pawn|Components|CharacterMovement|GetCurrentAcceleration (Variables|/Script/Engine.Character|GetCharacterMovement _asfort_player_pawn))))
    (bind _returnvalue (|CallParentFunction (Variables|GetInDirection)))
    (return _returnvalue)
    (else
      (bind _output_get (LocalVariables|SetAccel%20Dir (Math|Vector|Normalize (Pawn|Components|CharacterMovement|GetCurrentAcceleration (Variables|/Script/Engine.Character|GetCharacterMovement _asfort_player_pawn)))))
      (if (Math|Float|InRange(Float) (Math|Trig|Acos(Degrees) (Math|Vector|DotProduct (Variables|GetInDirection) _output_get)) (* (LocalVariables|GetMax%20Jump%20Angle) -1.0) (LocalVariables|GetMax%20Jump%20Angle))
        (return (LocalVariables|GetAccel%20Dir))
        (else
          (LocalVariables|SetTransform (Transformation|GetActorTransform (CallFunction|GetOwnerPawn)))
          (if (>= (.x (Math|Transform|InverseTransformDirection (LocalVariables|GetTransform) (LocalVariables|GetAccel%20Dir))) 0)
            (bind _returnvalue_2 (|CallParentFunction (Variables|GetInDirection)))
            (return _returnvalue_2)
            (else
              (bind _returnvalue_3 (|CallParentFunction (Variables|GetInDirection)))
              (bind (_returnvalue_x _returnvalue_y _returnvalue_z) (Math|Vector|BreakVector (Math|Transform|InverseTransformDirection (LocalVariables|GetTransform) _returnvalue_3)))
              (return (Math|Transform|TransformDirection (LocalVariables|GetTransform) (Math|Vector|MakeVector (* _returnvalue_x -1.0) _returnvalue_y _returnvalue_z))))))))))
